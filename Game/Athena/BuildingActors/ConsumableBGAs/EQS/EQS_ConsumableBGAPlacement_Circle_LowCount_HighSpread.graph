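# Environment query that picks placement points for consumable BGAs (building actors) in Athena, as a low-count, high-spread circle variant.
# A donut generator makes a sparse three-ring spiral around the querier, with radii bound to query params, and projects points onto geometry with a vertical offset.
# Square-root random scoring varies the choice, while a more heavily weighted flat-surface test penalizes uneven spots through a negative score rather than filtering them.
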EnvQuery /Game/Athena/BuildingActors/ConsumableBGAs/EQS/EQS_ConsumableBGAPlacement_Circle_LowCount_HighSpread
options [{
  generator {
   name=EnvQueryGenerator_Donut_0
   class=EnvQueryGenerator_Donut
   properties {
    bUseSpiralPattern=true
    center=&EnvQueryContext_Querier
    numberOfRings=3
    pointsPerRing=10
    innerRadius {
     dataBinding=&self:EnvQueryGenerator_Donut_0.AIDataProvider_QueryParams_4
     dataField=FloatValue
     defaultValue=1024
    }
    outerRadius {
     dataBinding=&self:EnvQueryGenerator_Donut_0.AIDataProvider_QueryParams_5
     dataField=FloatValue
     defaultValue=3072
    }
    projectionData {extentX=0, navigationFilter=null, postProjectionVerticalOffset=65, projectDown=3072, projectUp=256, traceMode=GeometryByChannel}
   }
  }
  tests [
   FortQueryTest_Random_1 {class=FortQueryTest_Random, properties {bUseRandomSeedForOthers=true, floatValueMax=1, scoringEquation=SquareRoot, scoringFactor=1}}
   FortQueryTest_OnFlatSurface_1 {
    class=FortQueryTest_OnFlatSurface
    properties {boolValue=false, radius=50, scoringFactor=-2, testPurpose=Score, toleranceZ=50, traceOffsetDown=50, traceOffsetUp=50}
   }
  ]
 }]
